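# Lit orb material, apparently for a Starlight mission dungeon totem, with a dark shell around glowing animated clouds and lightning.
# Two panning cloud samples, parallaxed by bump offset except on mobile, mask lightning whose uniform flicker comes from zero-tiled noise.
# The pattern drives roughness and, through UV cap and fresnel masks, emissive glow combined with rim and core fresnels.
# Driver scalars control panning, inner glow and blending to a dimmer off state.
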Material /Game/Missions/Primary/Starlight/Particles/M_DungTotem_Orb_02
graph {
 properties {
  bUsedWithMeshDeformer=true
  bUsedWithNanite=true
  bUsedWithNiagaraMeshParticles=true
  bUsedWithSkeletalMesh=true
  pixelDepthOffsetMode=PDOM_Legacy
  usedShadingModels=Default Lit
 }
 nodes [
  n1 = ScalarParameter() {defaultValue=1, group=Spec_Rough_adjust, parameterName=Roughness_pits}
  n2 = ScalarParameter() {defaultValue=0.4, group=Spec_Rough_adjust, parameterName=Roughness_low}
  n3 = ScalarParameter() {defaultValue=1, group=Spec_Rough_adjust, parameterName=Roghness_high}
  n4 = AppendVector(A=Constant() {r=1}, B=Constant() {r=1.1})
  n5 = TextureCoordinate() {vTiling=0.5}
  n6 = TextureSampleParameter2D(UVs=Panner(Coordinate=Multiply(A=@n5, B=@n4)) {speedX=0.0125, speedY=-0.035}) {
   group=InsideElements
   parameterName=CloudTexture
   samplerType=SAMPLERTYPE_LinearGrayscale
   texture=/Game/Global/Textures/Default/Tileable_Patterns/T_StormCloudDetail01_M
  }
  n7 = TextureSampleParameter2D(UVs=Panner(Coordinate=@n5) {speedX=0.07, speedY=0.025}) {
   group=InsideElements
   parameterName=CloudTexture
   samplerType=SAMPLERTYPE_LinearGrayscale
   texture=&T_StormCloudDetail01_M
  }
  n8 = Clamp(None=Power(
    Base=Multiply(A=@n6.R, B=@n7.R)
    Exp=ScalarParameter() {defaultValue=2.5, group=InsideElements, parameterName=CloudDepth_BumpDetail_Exponent}
   ) {constExponent=3})
  n9 = ScalarParameter() {defaultValue=-2, group=InsideElements, parameterName=CloudDepth_BumpOffset}
  n10 = ShadingPathSwitch(
   Default=BumpOffset(
    Coordinate=@n5
    Height=Multiply(A=@n8, B=@n9)
    HeightRatioInput=ScalarParameter() {defaultValue=0.11, group=InsideElements, parameterName=CloudDepth_HeightRatio}
   ) {heightRatio=0.1}
   Mobile=@n5
  ) {desc=Migrated from FeatureLevelSwitch}
  n11 = AppendVector(A=Constant() {r=1}, B=Constant() {r=1.1})
  n12 = AppendVector(
   A=ScalarParameter() {defaultValue=0.0125, group=InsideElements, parameterName=XPan-CloudsA}
   B=ScalarParameter() {defaultValue=-0.035, group=InsideElements, parameterName=YPan-CloudsA}
  )
  n13 = Clamp(None=ScalarParameter() {defaultValue=1, group=Driver, parameterName=Panning-OnOff, sliderMax=1})
  n14 = TextureSampleParameter2D(UVs=Panner(Coordinate=Multiply(A=@n10, B=@n11), Speed=LinearInterpolate(B=@n12, Alpha=@n13)) {speedX=0.0125, speedY=-0.035}) {
   group=InsideElements
   parameterName=CloudTexture
   samplerType=SAMPLERTYPE_LinearGrayscale
   texture=&T_StormCloudDetail01_M
  }
  n15 = AppendVector(
   A=ScalarParameter() {defaultValue=0.07, group=InsideElements, parameterName=XPan-CloudsB}
   B=ScalarParameter() {defaultValue=0.025, group=InsideElements, parameterName=YPan-CloudsB}
  )
  n16 = TextureSampleParameter2D(UVs=Panner(Coordinate=@n10, Speed=LinearInterpolate(B=@n15, Alpha=@n13)) {speedX=0.07, speedY=0.025}) {
   group=InsideElements
   parameterName=CloudTexture
   samplerType=SAMPLERTYPE_LinearGrayscale
   texture=&T_StormCloudDetail01_M
  }
  n17 = ScalarParameter() {defaultValue=5, group=OffState, parameterName=Offstate-CloudTexture_Power}
  n18 = ScalarParameter() {defaultValue=2, group=InsideElements, parameterName=CloudTexture_Power}
  n19 = Clamp(None=ScalarParameter() {defaultValue=1, group=Driver, parameterName=Functionality-OnOff, sliderMax=1})
  n20 = Clamp(None=Power(Base=Multiply(A=@n14.R, B=@n16.R), Exp=LinearInterpolate(A=@n17, B=@n18, Alpha=@n19)))
  n21 = TextureCoordinate()
  n22 = BumpOffset(
   Coordinate=@n21
   Height=Multiply(A=@n8, B=ScalarParameter() {defaultValue=-3, group=InsideElements, parameterName=LightningDepth_BumpOffset})
  ) {heightRatio=0.025, referencePlane=0}
  n23 = TextureSampleParameter2D(UVs=Panner(Coordinate=ShadingPathSwitch(Default=@n22, Mobile=@n21) {desc=Migrated from FeatureLevelSwitch}) {speedX=0.013, speedY=0.05}) {
   group=InsideElements
   parameterName=LightningMask_Texture
   samplerSource=SSM_Wrap_WorldGroupSettings
   texture=/Game/Weapons/WeaponSkins/Wraps/Textures/PlasmaClouds/T_Wrap_PlasmaClouds__M
  }
  n24 = Clamp(None=Power(Base=@n23.G, Exp=ScalarParameter() {defaultValue=2, group=InsideElements, parameterName=LightningTexture_Power}) {constExponent=4})
  n25 = TextureSample(UVs=Panner(Coordinate=TextureCoordinate() {uTiling=0, vTiling=0}) {speedX=0.015, speedY=0.3}) {
   samplerSource=SSM_Wrap_WorldGroupSettings
   samplerType=SAMPLERTYPE_LinearColor
   texture=/Game/Characters/Player/Male/Medium/Bodies/M_MED_SRDrift/Textures/T_ElectricNoise
  }
  n26 = Multiply(
   A=Power(Base=@n25.G, Exp=ScalarParameter() {defaultValue=1, group=InsideElements, parameterName=LightningFlickerPower})
   B=LinearInterpolate(
    A=ScalarParameter() {defaultValue=250, group=OffState, parameterName=Offstate-LightningMaxBright}
    B=ScalarParameter() {defaultValue=650, group=InsideElements, parameterName=LightningMaxBright}
    Alpha=@n19
   )
  )
  n27 = ScalarParameter() {defaultValue=4, group=InsideElements, parameterName=LightningMask_Power}
  n28 = Add(A=@n20, B=Multiply(A=Multiply(A=@n24, B=@n26), B=Power(Base=@n20, Exp=@n27)))
  n29 = Clamp(None=OneMinus(None=@n28)) {minDefault=0.5}
  n30 = Power(
   Base=Clamp(None=LinearInterpolate(Alpha=@n29) {constA=-1})
   Exp=ScalarParameter() {defaultValue=1.3809, group=Spec_Rough_adjust, parameterName=Edge_contrast}
  )
  n31 = ScalarParameter() {defaultValue=1.3107, group=Spec_Rough_adjust, parameterName=Edge_Stregnth}
  n32 = LinearInterpolate(
   A=Clamp(None=LinearInterpolate(A=@n2, B=@n3))
   B=ScalarParameter() {group=Spec_Rough_adjust, parameterName=Roughness_Edge}
   Alpha=Clamp(None=Multiply(A=@n30, B=@n31))
  )
  n33 = OneMinus(None=@n28)
  n34 = Clamp(None=LinearInterpolate(Alpha=Clamp(None=@n33) {maxDefault=0.5}) {constB=2})
  n35 = ScalarParameter() {defaultValue=1.3262, group=Spec_Rough_adjust, parameterName=Depth_Contrast}
  n36 = Clamp(None=Multiply(A=Power(Base=@n34, Exp=@n35), B=ScalarParameter() {defaultValue=1.2614, group=Spec_Rough_adjust, parameterName=Depth_intensity}))
  LinearInterpolate(
   A=Clamp(None=LinearInterpolate(A=@n1, B=@n32, Alpha=@n36))
   B=ScalarParameter() {defaultValue=0.15, group=Roughness-Simple, parameterName=SimpleRoughness}
   Alpha=ScalarParameter() {group=Roughness-Simple, parameterName=UseSimpleRoughness}
  )
  LinearInterpolate(
   A=ScalarParameter() {defaultValue=0.5, group=Side Projection, parameterName=SpecularMin, sortPriority=1}
   B=ScalarParameter() {defaultValue=0.1, group=Side Projection, parameterName=SpecularMax, sortPriority=1}
  )
  n37 = ScalarParameter() {defaultValue=1, group="Main \"Shell\"", parameterName=Normal Tile X}
  n38 = ScalarParameter() {defaultValue=1, group="Main \"Shell\"", parameterName=Normal Tile Y}
  n39 = TextureSampleParameter2D(UVs=Multiply(A=TextureCoordinate() {uTiling=0, vTiling=0}, B=AppendVector(A=@n37, B=@n38))) {
   group="Main \"Shell\""
   parameterName=Normal_Texture
   samplerType=SAMPLERTYPE_Normal
   texture=&FlatNormalBlue
  }
  MaterialFunctionCall(Normal=@n39.RGB, Flatness=ScalarParameter() {group="Main \"Shell\"", parameterName="Normal Flatness (S)"}) {materialFunction=&FlattenNormal}
  n40 = VectorParameter() {group=InsideElements, parameterName=Color-Clouds, defaultValue=<rgba 1.9688 1.9813 2 1>}
  n41 = LinearInterpolate(A=ScalarParameter() {defaultValue=0.2, group=OffState, parameterName=Offstate_ColorMultiply}, Alpha=@n19)
  n42 = ComponentMask(None=TextureCoordinate() {uTiling=0.5, vTiling=0.5}) {g=true}
  n43 = Multiply(
   A=Power(Base=Sine(None=@n42), Exp=ScalarParameter() {defaultValue=3, group=UV Cap Mask, parameterName=GradMask_Exponent}) {constExponent=3}
   B=ScalarParameter() {defaultValue=3, group=UV Cap Mask, parameterName=GradMask_Hardness}
  ) {constB=3}
  n44 = Multiply(
   A=@n28
   B=LinearInterpolate(
    A=Constant() {r=1}
    B=Clamp(None=@n43)
    Alpha=ScalarParameter() {defaultValue=1, group=UV Cap Mask, parameterName=UseUVCapMask?, sliderMax=1}
   )
  )
  n45 = MaterialFunctionCall(
   Brightness=ScalarParameter() {defaultValue=1, group=InsideElements, parameterName=CloudFresnel_Brightness}
   Exponent=LinearInterpolate(
    A=ScalarParameter() {defaultValue=10, group=OffState, parameterName=Offstate-CloudFresnel_Exponent}
    B=ScalarParameter() {defaultValue=6, group=InsideElements, parameterName=CloudFresnel_Exponent}
    Alpha=@n19
   )
  ) {materialFunction=&MF_Fresnel_FX}
  n46 = Multiply(
   A=Multiply(A=Multiply(A=@n40.RGB, B=@n41), B=Multiply(A=@n44, B=@n45.Fresnel))
   B=ScalarParameter() {defaultValue=1, group=Driver, parameterName=InnerGlow, sliderMax=1}
  )
  n47 = Multiply(
   A=VectorParameter() {group="Main \"Shell\"", parameterName=Color_RimFresnel, defaultValue=<rgba 0.0684 0.371 0.625 1>}.RGB
   B=MaterialFunctionCall(
    Brightness=ScalarParameter() {defaultValue=1, group="Main \"Shell\"", parameterName=RimFresnel_Brightness}
    Exponent=ScalarParameter() {defaultValue=4, group="Main \"Shell\"", parameterName=RimFresnel_Exponent}
   ) {materialFunction=&MF_Fresnel_FX}.Fresnel Inverted
  )
  n48 = MaterialFunctionCall(
   Brightness=ScalarParameter() {defaultValue=1, group="Main \"Shell\"", parameterName=CoreFresnel_Brightness}
   Exponent=ScalarParameter() {defaultValue=4, group="Main \"Shell\"", parameterName=CoreFresnel_Exponent}
  ) {materialFunction=&MF_Fresnel_FX}
  n49 = VectorParameter() {group="Main \"Shell\"", parameterName=CoreFresnel_Color, defaultValue=<rgba 0 0.5337 1 1>}
  MaterialFunctionCall(Emissive=Add(A=Add(A=@n46, B=@n47), B=Multiply(A=@n48.Fresnel Unclamped, B=@n49.RGB))) {materialFunction=&MF_FX_EmissiveAdjustments}
  Reroute(None=ScalarParameter() {group="Main \"Shell\"", parameterName=Metallic})
  Reroute(None=VectorParameter() {group="Main \"Shell\"", parameterName=Color_Base, defaultValue=<rgba 0.0025 0 0.01 1>}.RGB)
 ]
}
